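# Niagara system for the EagerUnify backpack cosmetic, drawing green pulsing glows on its body and ears plus a flickering ribbon tail.
# Two sprite emitters burst fixed points placed by partition index, with gender offset and camera-angle-driven size and fade.
# TailTrail lays ten wobbling ribbon points along an axis set by IsSkydiving, blending clamped, smoothed world-space lag toward the tip.
# A disabled debug emitter visualizes the bIsMale flag.
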NiagaraSystem /BRCosmetics/Accessories/FORT_Backpacks/Backpack_F_MED_EagerUnify/FX/NS_BackPack_EagerUnify
dependencies {
 usedDataInterfaces [&NiagaraDataInterfaceCamera, &NiagaraDataInterfaceCurve, &NiagaraDataInterfaceSocketReader]
 usedDynamicInputs [&VectorLength, &SetBoolByFloatComparison]
 usedModules [
  &self:SystemSpawnScript.SystemScriptSource.SystemScriptGraph.NiagaraNodeAssignment_2.SetVariables_#1
  &SystemState
  &self:SystemSpawnScript.SystemScriptSource.SystemScriptGraph.NiagaraNodeAssignment_3.SetVariables_#2
  &Timeline
  &self:SystemSpawnScript.SystemScriptSource.SystemScriptGraph.NiagaraNodeAssignment_4.SetVariables_#3
 ]
}
emitters {
 EarGlows {
  emitterUpdateScript [
   &EmitterState(
    Inactive Response=::NewEnumerator1
    Life Cycle Mode=::NewEnumerator1
    Loop Behavior=::NewEnumerator1
    Loop Duration Mode=::NewEnumerator1
    UseLoopDelay=false
   )
   &SpawnBurst_Instantaneous(Spawn Count=2, Use Spawn Probability=false, UseLoopCountLimit=true)
   /SetVariables(Emitter.SideLerp:NiagaraFloat=&Abs_Float(Float:NiagaraFloat=&DotProduct(Vector A=&ReturnCameraVector(Camera Vectors=::NewEnumerator1, Coordinate Space=::Simulation), Vector B=<1 0 0>))) {isSetParameters=true}
  ]
  particleSpawnScript [
   &PartitionParticles(ID to Partition:NiagaraInt32=&ReturnExecIndex())
   &InitializeParticle(
    Color Mode=::NewEnumerator0
    Position Offset=&AddVectorToPosition(
     Position=&MergeHelper_VectorSwitch(0=<6.9 -15.7 10.9>, 1=<-6.9 -15.7 10.9>, 2=?, 3=?, Index=$Particles.PartitionParticles.Partition, Num=::NewEnumerator0)
     Vector=$System.OffsetByGender
    )
    UsePositionOffset=true
    Write Material Random=true
   )
   /SetVariables() {isSetParameters=true}
  ]
  particleUpdateScript [
   &ParticleState(Kill Particles When Lifetime Has Elapsed=false, Let Infinitely Lived Particles Die When Emitter Deactivates=true)
   /SetVariables(Particles.MainColor:LinearColor=&MultiplyLinearColor(
     Linear Color=<rgba 0.1017 1 0.02 1>
     Scale Factor=&MakeVector4FromVectorAndFloat(
      "Vector (RGB)"=&VectorFromFloat(Value=&Waveform(
        Waveform Count=::NewEnumerator1
        Global Input Position=$Particles.PartitionParticles.NormalizedPartition
        "Waveform [1]"=::NewEnumerator12
        "[1] Frequency"=6
        "[1] Amplitude Min/Max"=<0.8 1>
        "[1] Bias"=0.2
        "[2] Average"=false
        "[2] Input Position"=0
        "[2] Phase"=$Emitter.Age
        "[2] Phase Scale"=1
        "[2] Compound Frequency"=2
        "[2] Frequency"=1.2
        "[2] Amplitude Min/Max"=<0 0.2>
        "[2] Scale"=1
        "[2] Bias"=0
        "[2] Exponent"=1
        "[2] Pulse Duration"=0.1
        "[2] Clip Triangle Min/Max"=<0 0>
        "[2] Random Seed"=0
       ))
      "Float (Alpha)"=0.6
     )
    )) {isSetParameters=true}
   /SetVariables(
    Particles.SoftGlowCameraOffset=3
    Particles.SoftGlowColor:LinearColor=&MultiplyLinearColor(
     Linear Color=$StackContext.MainColor
     Scale Factor=&MakeVector4_FromFloat(Float=&MergeHelper_FloatSwitch(0=0.4, 1=0.4, 2=0.2, 3=0.2, 4=0.5, Index=$Particles.PartitionParticles.Partition, Num=::NewEnumerator3))
    )
    Particles.SoftGlowSize:Vector2f=&Vector2DFromFloat(Value=&MergeHelper_FloatSwitch(0=14, 1=14, 2=13, 3=13, 4=13, Index=$Particles.PartitionParticles.Partition, Num=::NewEnumerator3))
   ) {isSetParameters=true}
   &CameraOffset()
  ]
  rendererClasses [&NiagaraSpriteRendererProperties]
  renderers [{rendererClass=&NiagaraSpriteRendererProperties}]
 }
 FN_Debug {
  bEnabled=false
  emitterUpdateScript [
   &EmitterState(Life Cycle Mode=::NewEnumerator1, Loop Behavior=::NewEnumerator1, Loop Duration=1, UseLoopDelay=false)
   &SpawnBurst_Instantaneous(Spawn Count=1, Use Spawn Probability=false)
  ]
  particleSpawnScript [
   /Niagara/Modules/Spawn/Initialization/InitializeParticle(
    Lifetime=5
    Module.Write Color=true
    Module.Write Lifetime=true
    Module.Write Mass=true
    Module.Write Material Random=true
    Module.Write Position=false
    Module.Write Scale=false
    Module.Write SpriteRotation=false
    Module.Write SpriteSize=true
    Position=&SimulationPosition()
    Sprite Size=<100 100>
   )
   &SystemLocation(MaskSpawnGroup=false, Offset=<-100 0 0>)
  ]
  particleUpdateScript [
   &ParticleState()
   /SetVariables(Particles.Debug:NiagaraFloat=&MakeCustomFloatFromBool(Bool=$User.bIsMale, True Float=1, False Float=0)) {isSetParameters=true}
   &DynamicMaterialParameters(
    Index 0 Param 1=$Particles.Debug
    Module.Param0WriteEnabled=true
    Module.Param1WriteEnabled=true
    Module.Param2WriteEnabled=true
    Module.Param3WriteEnabled=true
   )
  ]
  rendererClasses [&NiagaraSpriteRendererProperties]
  renderers [{rendererClass=&NiagaraSpriteRendererProperties}]
 }
 GlowsAndLights {
  emitterUpdateScript [
   &EmitterState(
    Inactive Response=::NewEnumerator1
    Life Cycle Mode=::NewEnumerator1
    Loop Behavior=::NewEnumerator1
    Loop Duration Mode=::NewEnumerator1
    UseLoopDelay=false
   )
   &SpawnBurst_Instantaneous(Spawn Count=6, Use Spawn Probability=false, UseLoopCountLimit=true)
   /SetVariables(Emitter.FrontLerp:NiagaraFloat=&Abs_Float(Float:NiagaraFloat=&DotProduct(Vector A=&ReturnCameraVector(Camera Vectors=::NewEnumerator1, Coordinate Space=::Simulation), Vector B=<0 1 0>))) {isSetParameters=true}
  ]
  particleSpawnScript [
   &PartitionParticles(ID to Partition:NiagaraInt32=&ReturnExecIndex(), Number Of Partitions=6)
   /Niagara/Modules/Spawn/Initialization/V2/InitializeParticle(
    Color Mode=::NewEnumerator0
    Position Offset=&AddVectorToPosition(
     Position=&MergeHelper_VectorSwitch(
      0=<5.0633 -13.1858 -13.0573>
      1=<-5.0633 -13.1858 -13.0573>
      2=<9.6713 -10.638 -6.1296>
      3=<-9.6713 -10.638 -6.1296>
      4=<0 -0.6913 -8.5622>
      5=<-0.0452 -2.8882 -10.5>
      Index=$Particles.PartitionParticles.Partition
      Num=::NewEnumerator4
     )
     Vector=$System.OffsetByGender
    )
    UsePositionOffset=true
    Write Material Random=true
   )
   /SetVariables() {isSetParameters=true}
  ]
  particleUpdateScript [
   &ParticleState(Kill Particles When Lifetime Has Elapsed=false, Let Infinitely Lived Particles Die When Emitter Deactivates=true)
   &UpdateMeshOrientation(Facing Interpolation=false, Orientation Method=::NewEnumerator1, Override Quat=false, TransformFacing=false, TransformUp=false)
   /SetVariables(Particles.MainColor:LinearColor=&MultiplyLinearColor(
     Linear Color=<rgba 0.1017 1 0.02 1>
     Scale Factor=&MakeVector4FromVectorAndFloat(
      "Vector (RGB)"=&VectorFromFloat(Value=&Waveform(
        Waveform Count=::NewEnumerator1
        Global Input Position=$Particles.PartitionParticles.NormalizedPartition
        "Waveform [1]"=::NewEnumerator12
        "[1] Frequency"=6
        "[1] Amplitude Min/Max"=<0.35 1>
        "[1] Bias"=0.2
        "[2] Average"=false
        "[2] Input Position"=0
        "[2] Phase"=$Emitter.Age
        "[2] Phase Scale"=1
        "[2] Compound Frequency"=2
        "[2] Frequency"=1.2
        "[2] Amplitude Min/Max"=<0 0.2>
        "[2] Scale"=1
        "[2] Bias"=0
        "[2] Exponent"=1
        "[2] Pulse Duration"=0.1
        "[2] Clip Triangle Min/Max"=<0 0>
        "[2] Random Seed"=0
       ))
      "Float (Alpha)"=0.8
     )
    )) {isSetParameters=true}
   /SetVariables(Particles.BottomFade:NiagaraFloat=&RemapRange(
     Input Value To Remap=&ClampFloat(Float=&DotProduct(Vector A=&ReturnCameraVector(Camera Vectors=::NewEnumerator1, Coordinate Space=::Simulation), Vector B=<0 0 1>))
     Input Min=0.8
     Input Max=1
     Output Min=1
     Output Max=0
    )) {isSetParameters=true}
   /SetVariables(
    Particles.SoftGlowCameraOffset=3
    Particles.SoftGlowColor:LinearColor=&MultiplyLinearColor(
     Linear Color=$StackContext.MainColor
     Scale Factor=&MakeVector4_FromFloat(Float=&MergeHelper_FloatSwitch(
       0=0.25
       1=0.25
       2=0.25
       3=0.25
       4=0.25
       5=&Lerp_Float(A=1, B=0.7, Alpha=$StackContext.BottomFade)
       Index=$Particles.PartitionParticles.Partition
       Num=::NewEnumerator4
      ))
    )
    Particles.SoftGlowSize:Vector2f=&Vector2DFromFloat(Value=&MergeHelper_FloatSwitch(
      0=7
      1=7
      2=10
      3=10
      4=7
      5=&Lerp_Float(A=14, B=20, Alpha=$StackContext.BottomFade)
      Index=$Particles.PartitionParticles.Partition
      Num=::NewEnumerator4
     ))
   ) {isSetParameters=true}
   /SetVariables(
    Particles.CapsuleGlowColor:LinearColor=&MultiplyLinearColor(Linear Color=$StackContext.MainColor, Scale Factor=<1 1 1 0.333>)
    Particles.CapsuleSpriteSize:Vector2f=&Vector2DFromFloat(Value=&MergeHelper_FloatSwitch(
      0=5
      1=5
      2=&Lerp_Float(A=8, B=5, Alpha=$Emitter.FrontLerp)
      3=&Lerp_Float(A=8, B=5, Alpha=$Emitter.FrontLerp)
      4=4
      5=6
      Index=$Particles.PartitionParticles.Partition
      Num=::NewEnumerator4
     ))
   ) {isSetParameters=true}
   /SetVariables(Particles.CapsuleSpriteOrientation:Quat4f=/Game/Effects/Niagara/Modules/MergeEmitters/MergeHelper_QuatSwitch(
     0:Quat4f=&MakeQuaternion(Quaternion From=2, Axis=<1 0 0>, Angle=0, X=-40.9885, Z=38.0051, From=?, To=?, Angle Type=::NewEnumerator0, Normalize Vectors=false)
     1:Quat4f=&MakeQuaternion(Quaternion From=2, Axis=<1 0 0>, Angle=0, X=-40.9885, Z=-38.0051, From=?, To=?, Angle Type=::NewEnumerator0, Normalize Vectors=false)
     2:Quat4f=&MakeQuaternion(
      Quaternion From=2
      Axis=<1 0 0>
      Angle=0
      X=134.0506
      Y=61.5028
      Z=25.9275
      From=?
      To=?
      Angle Type=::NewEnumerator0
      Normalize Vectors=false
     )
     3:Quat4f=&MakeQuaternion(
      Quaternion From=2
      Axis=<1 0 0>
      Angle=0
      X=134.0506
      Y=-61.5028
      Z=-25.9275
      From=?
      To=?
      Angle Type=::NewEnumerator0
      Normalize Vectors=false
     )
     4:Quat4f=&MakeQuaternion(Quaternion From=2, Axis=<1 0 0>, Angle=0, X=21.9503, From=?, To=?, Angle Type=::NewEnumerator0, Normalize Vectors=false)
     5:Quat4f=&MakeQuaternion(Quaternion From=2, Axis=<1 0 0>, Angle=0, X=43.6358, From=?, To=?, Angle Type=::NewEnumerator0, Normalize Vectors=false)
     6:Quat4f=?
     7:Quat4f=?
     8:Quat4f=?
     9:Quat4f=?
     Index:NiagaraInt32=$Particles.PartitionParticles.Partition
     Num:MergeHelper_SwitcherEnum=::NewEnumerator4
    )) {isSetParameters=true}
   /SetVariables(Particles.CapsuleSpriteAlignment:Vector3f=&TransformVector(
     Vector=&ReturnMeshOrientationAxis(Mesh Orientation Axis=::XAxis, Mesh Orientation Quat=$StackContext.CapsuleSpriteOrientation)
     Destination Space=::World
    )) {isSetParameters=true}
   /SetVariables(Particles.CapsuleSpriteDynamic:Vector4f=&MakeFloat4FromVectorAndFloat(
     "Vector (RGB)"=&NormalizeVector(Vector To Normalize=$Particles.CapsuleSpriteAlignment, Safe Normalize=true, Safe Fallback Threshold=0, Safe Fallback Vector=<0 0 0.001>)
     "Float (Alpha)"=&MergeHelper_FloatSwitch(
      0=10
      1=10
      2=&Lerp_Float(A=11, B=8, Alpha=$Emitter.FrontLerp)
      3=&Lerp_Float(A=11, B=8, Alpha=$Emitter.FrontLerp)
      4=8
      5=6
      Index=$Particles.PartitionParticles.Partition
      Num=::NewEnumerator4
     )
    )) {isSetParameters=true}
   /SetVariables(Particles.MeshOrientation=$StackContext.CapsuleSpriteOrientation) {enabled=false, isSetParameters=true}
   &CameraOffset()
  ]
  rendererClasses [&NiagaraSpriteRendererProperties]
  renderers [
   {rendererClass=&NiagaraSpriteRendererProperties}
   {rendererClass=&NiagaraSpriteRendererProperties, rendererIndex=1}
  ]
 }
 TailTrail {
  emitterSpawnScript [/SetVariables(
    Emitter.RibbonLength:NiagaraFloat=&Multiply_Float(A=15, B=-1)
    Emitter.StartPos:Vector3f=<-0.0452 -2.8882 -10>
   ) {isSetParameters=true}]
  emitterUpdateScript [
   &EmitterState(
    Inactive Response=::NewEnumerator1
    Life Cycle Mode=::NewEnumerator1
    Loop Behavior=::NewEnumerator1
    Loop Duration Mode=::NewEnumerator1
    UseLoopDelay=false
   )
   &SpawnBurst_Instantaneous(Spawn Count=10, Use Spawn Probability=false, UseLoopCountLimit=false)
  ]
  particleSpawnScript [
   &InitializeParticle(
    Color=<rgba 1.3 1.3 1.3 1>
    Lifetime=5
    Mesh Scale Mode=::NewEnumerator3
    Mesh Uniform Scale=2
    Sprite Size Mode=::NewEnumerator3
    Uniform Sprite Size=3
    UsePositionOffset=true
    Write Material Random=true
   )
   /SetVariables(Particles.ParticleOrder:NiagaraFloat=&ReturnNormalizedExecIndex()) {isSetParameters=true}
   /SetVariables() {isSetParameters=true}
  ]
  particleUpdateScript [
   &ParticleState(
    Kill Particles When Lifetime Has Elapsed=false
    Let Infinitely Lived Particles Die When Emitter Deactivates=true
    Loop Particles Lifetime=true
   )
   /SetVariables(Particles.GlitchCurve:NiagaraFloat=&Waveform("Waveform [1]"=::NewEnumerator0, "[1] Phase Scale"=0.1, "[1] Frequency"=1, "[1] Amplitude Min/Max"=<0 1>, "[1] Exponent"=10)) {isSetParameters=true}
   /SetVariables(Particles.TimeOffset:NiagaraFloat=&Add_Float(
     A=$Particles.TimeOffset
     B=&Multiply_Float(
      A=$Engine.DeltaTime
      B=&Multiply_Float(
       A=&Lerp_Float(A=1, B=1.5, Alpha=$StackContext.GlitchCurve)
       B=&Lerp_Float(A=1, B=2, Alpha=$System.Timeline.Time)
      )
     )
    )) {isSetParameters=true}
   /SetVariables(Particles.Position=&AddVectorToPosition(
     Position=&Lerp_Vector(
      A=$Emitter.StartPos
      B=&Add_Vector(
       A=$Emitter.StartPos
       B=&Multiply_VectorByFloat(
        Vector=&RotateVectorByQuaternion(
         Vector To Rotate=&TransformVector(
          Vector=&MakeCustomVectorFromBool(Bool=$User.IsSkydiving, True Vector=<0 1 1>, False Vector=<0 0 1>)
          Source Space=::World
          Destination Space=::Local
         )
         Quaternion=$Engine.Owner.Rotation
        )
        Float=$Emitter.RibbonLength
       )
      )
      Alpha=$Particles.ParticleOrder
     )
     Vector=$System.OffsetByGender
    )) {isSetParameters=true}
   /SetVariables(Particles.BottomFade:NiagaraFloat=&RemapRange(
     Input Value To Remap=&ClampFloat(Float=&DotProduct(Vector A=&ReturnCameraVector(Camera Vectors=::NewEnumerator1, Coordinate Space=::Simulation), Vector B=<0 0 1>))
     Input Min=0.8
     Input Max=1
     Output Min=1
     Output Max=0
    )) {isSetParameters=true}
   /SetVariables(Particles.Position=&AddVectorToPosition(
     Position=$StackContext.Position
     Vector=&MakeVector(
      X=&Waveform(
       Global Amplitude Scale=&FloatFromCurve(FloatCurve=DataInterface() {propertyValues {Curve {keys [<at 0 0>,<at 1 1>]}}}, CurveIndex=$Particles.ParticleOrder)
       "Waveform [1]"=::NewEnumerator0
       "[1] Input Position"=$Particles.ParticleOrder
       "[1] Phase Scale"=0.5
       "[1] Frequency"=1
       "[1] Amplitude Min/Max"=<-1 1>
      )
      Y=&MakeFloatFromVector(
       Vector=&MakeVector(
        X=&Waveform(
         Global Amplitude Scale=&FloatFromCurve(FloatCurve=DataInterface() {propertyValues {Curve {keys [<at 0 0>,<at 1 1>]}}}, CurveIndex=$Particles.ParticleOrder)
         "Waveform [1]"=::NewEnumerator1
         "[1] Input Position"=$Particles.ParticleOrder
         "[1] Phase Scale"=0.35
         "[1] Frequency"=1
         "[1] Amplitude Min/Max"=<-1 1>
         "[1] Scale"=0.5
        )
        Y=0
        Z=0
       )
       Channel=::NewEnumerator0
      )
      Z=0
     )
    )) {isSetParameters=true}
   &InterpolateOverTime(
    Rate Of Change:NiagaraFloat=&MakeCustomFloatFromBool(Bool=$User.IsSkydiving, True Float=25, False Float=20)
    Target Position Value:NiagaraPosition=&TransformPosition(Position=$Particles.Position, Source Space=::Local, Destination Space=::World)
    Variable Type=::NewEnumerator6
   )
   /SetVariables(Particles.Flicker:NiagaraFloat=&Lerp_Float(
     A=&Waveform("Waveform [1]"=::NewEnumerator8, "[1] Frequency"=12, "[1] Amplitude Min/Max"=<0.4 0.5>)
     B=&Waveform("Waveform [1]"=::NewEnumerator8, "[1] Frequency"=12, "[1] Amplitude Min/Max"=<0.8 1>)
     Alpha=$System.Timeline.Time
    )) {isSetParameters=true}
   /SetVariables(Particles.Position=&Lerp_Vector(
     A=Linked() {linkedVariable {name=StackContext.Position, type=NiagaraPosition}}
     B=&MakeCustomVectorFromBool(
      Bool=$User.IsSkydiving
      True Vector=&MakeVector(
       X=&ClampFloat(
        Float=&MakeFloatFromVector(
         Vector=&TransformPosition(Position=$StackContext.InterpolateOverTime.InterpolateOverTime_Position.Moving Average, Source Space=::World, Destination Space=::Local)
         Channel=::NewEnumerator0
        )
        Min=-200
        Max=200
       )
       Y=&MakeFloatFromVector(Vector=Linked() {linkedVariable {name=StackContext.Position, type=NiagaraPosition}}, Channel=::NewEnumerator1)
       Z=&MakeFloatFromVector(Vector=Linked() {linkedVariable {name=StackContext.Position, type=NiagaraPosition}}, Channel=::NewEnumerator2)
      )
      False Vector=&MakeVector(
       X=&ClampFloat(
        Float=&MakeFloatFromVector(
         Vector=&TransformPosition(Position=$StackContext.InterpolateOverTime.InterpolateOverTime_Position.Moving Average, Source Space=::World, Destination Space=::Local)
         Channel=::NewEnumerator0
        )
        Min=-100
        Max=100
       )
       Y=&ClampFloat(
        Float=&MakeFloatFromVector(
         Vector=&TransformPosition(Position=$StackContext.InterpolateOverTime.InterpolateOverTime_Position.Moving Average, Source Space=::World, Destination Space=::Local)
         Channel=::NewEnumerator1
        )
        Min=-7
        Max=80
       )
       Z=&Lerp_Float(
        A=&MakeFloatFromVector(
         Vector=&TransformPosition(Position=$StackContext.InterpolateOverTime.InterpolateOverTime_Position.Moving Average, Source Space=::World, Destination Space=::Local)
         Channel=::NewEnumerator2
        )
        B=&MakeFloatFromVector(Vector=Linked() {linkedVariable {name=StackContext.Position, type=NiagaraPosition}}, Channel=::NewEnumerator2)
        Alpha=0.8
       )
      )
     )
     Alpha=&Power(a=$Particles.ParticleOrder, b=2)
    )) {isSetParameters=true}
   /SetVariables(Particles.RibbonWidth=&Multiply_Float(
     A=&FloatFromCurve(FloatCurve=DataInterface() {propertyValues {Curve {keys [<at 0 0.5>,<at 0.3 1>,<at 1 0.2>]}}}, CurveIndex=$Particles.ParticleOrder)
     B=&Multiply_Float(A=6, B=&Lerp_Float(A=0.7, B=0.8, Alpha=&ClampFloat(Float=$Particles.GlitchCurve)))
    )) {isSetParameters=true}
   /SetVariables(
    Particles.RibbonGlowColor:LinearColor=&Multiply_LinearColorByFloat(Linear Color=<rgba 0.5333 2 0.4 0.8>, Float=&Multiply_Float(A=1, B=$StackContext.BottomFade))
    Particles.RibbonGlowWidth:NiagaraFloat=&Multiply_Float(A=$StackContext.RibbonWidth, B=2)
   ) {isSetParameters=true}
   &DynamicMaterialParameters(
    Index 0 Param 1=1
    Index 1 Param 1=1
    Index 1 Param 2=$Particles.TimeOffset
    Index 1 Param 3=0.7
    Index 1 Param 4=0.5
    Module.Index0Float0Write=true
    Module.Index0Float1Write=true
    Module.Index0Float2Write=true
    Module.Index0Float3Write=true
    Module.Index1Float0Write=false
    Module.Index1Float1Write=true
    Module.Index1Float2Write=true
    Module.Index1Float3Write=true
    Write Parameter Index 1=::NewEnumerator1
   )
   &ScaleColor(
    Scale Alpha=$Particles.BottomFade
    Scale RGB=&VectorFromFloat(Value=&Add_Float(A=$Particles.Flicker, B=1.3))
    ScaleA=true
    ScaleRGB=true
   )
  ]
  rendererClasses [&NiagaraRibbonRendererProperties]
  renderers [
   {rendererClass=&NiagaraRibbonRendererProperties}
   {rendererClass=&NiagaraRibbonRendererProperties, rendererIndex=1}
  ]
 }
}
summary {
 systemName=NS_BackPack_EagerUnify
 emitters [
  {emitterName=TailTrail, rendererClasses [&NiagaraRibbonRendererProperties]}
  {emitterName=GlowsAndLights, rendererClasses [&NiagaraSpriteRendererProperties]}
  {emitterName=EarGlows, rendererClasses [&NiagaraSpriteRendererProperties]}
  {bEnabled=false, emitterName=FN_Debug, rendererClasses [&NiagaraSpriteRendererProperties]}
 ]
 userVariables [
  User.bIsMale {type=NiagaraBool}
  User.IsSkydiving {type=NiagaraBool}
 ]
}
