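# Physics asset for the EagerUnify backpack cosmetic, setting up a dangling charm with two strands.
# Three large capsules approximate the pack body, and small capsules cover each charm bone.
# The charm base joins the pack with free swing and twist, then leads to dyn_charm_1, which branches into five-bone "a" and four-bone "b" chains.
# Chain joints have narrow swing limits and zero twist, with slightly tighter swing on the "b" strand.
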PhysicsAsset /BRCosmetics/Accessories/FORT_Backpacks/Backpack_F_MED_EagerUnify/Meshes/F_MED_EagerUnify_Pack_Physics
bodies {
 base_charm_1 [{shapeType=Capsule, length=0.505, radius=0.505, shapeName=base_charm_1_capsule, rotation=<rot 0 -90 0>}]
 dyn_charm_1 [{shapeType=Capsule, length=0.505, radius=0.505, shapeName=dyn_charm_1_capsule, center=<0.8723 0 0>, rotation=<rot 0 -90.0217 -90>}]
 dyn_charm_a_1 [{
   shapeType=Capsule
   length=0.505
   radius=0.505
   shapeName=dyn_charm_a_1_capsule
   center=<0.0997 -0.0038 0.0011>
   rotation=<rot 0 79.4958 -85.9832>
  }]
 dyn_charm_a_2 [{
   shapeType=Capsule
   length=0.505
   radius=0.505
   shapeName=dyn_charm_a_2_capsule
   center=<0.1169 -0.0176 0.0154>
   rotation=<rot 0 -162.7967 -89.9028>
  }]
 dyn_charm_a_3 [{
   shapeType=Capsule
   length=0.505
   radius=0.505
   shapeName=dyn_charm_a_3_capsule
   center=<0.1097 -0.0081 0.0101>
   rotation=<rot 0 177.3052 -48.6322>
  }]
 dyn_charm_a_4 [{
   shapeType=Capsule
   length=0.505
   radius=0.505
   shapeName=dyn_charm_a_4_capsule
   center=<0.12 -0.0017 -0.0115>
   rotation=<rot 0 -178.9525 -83.8535>
  }]
 dyn_charm_a_5 [{
   shapeType=Capsule
   length=1.1923
   radius=0.9209
   shapeName=dyn_charm_a_5_capsule
   center=<1.1858 -0.0612 0.0856>
   rotation=<rot 0 94.5744 -89.3469>
  }]
 dyn_charm_b_1 [{
   shapeType=Capsule
   length=0.505
   radius=0.505
   shapeName=dyn_charm_b_1_capsule
   center=<0.1221 -0.0289 -0.0045>
   rotation=<rot 0 84.8863 -79.974>
  }]
 dyn_charm_b_2 [{
   shapeType=Capsule
   length=0.505
   radius=0.505
   shapeName=dyn_charm_b_2_capsule
   center=<0.0913 0.0076 0.0085>
   rotation=<rot 0 -2.6072 -66.7047>
  }]
 dyn_charm_b_3 [{
   shapeType=Capsule
   length=0.505
   radius=0.505
   shapeName=dyn_charm_b_3_capsule
   center=<0.1132 -0.0252 0.0302>
   rotation=<rot 0 -179.4206 -52.3271>
  }]
 dyn_charm_b_4 [{
   shapeType=Capsule
   length=0.4705
   radius=0.5466
   shapeName=dyn_charm_b_4_capsule
   center=<0.5897 -0.1582 -0.02>
   rotation=<rot 0 -104.0934 -87.3858>
  }]
 pack [
  {shapeType=Capsule, length=24.6925, radius=7.5755, shapeName=pack_capsule_A, center=<-0.7319 -4.5218 -3.5412>}
  {
   shapeType=Capsule
   length=25.4637
   radius=4.5363
   shapeName=pack_capsule_B
   center=<-6.6619 -1.1344 -0.183>
   rotation=<rot 17.1549 -0.0001 -0.0004>
  }
  {shapeType=Capsule, length=23.3476, radius=6.6524, shapeName=pack_capsule_C, center=<-1.8768 -11.6447 0>}
 ]
}
constraints [
 {bone1Name=dyn_charm_1, bone2Name=base_charm_1, swing1LimitDegrees=20, swing2LimitDegrees=40, twistLimitDegrees=0}
 {bone1Name=dyn_charm_a_1, bone2Name=dyn_charm_1, swing1LimitDegrees=20, swing2LimitDegrees=40, twistLimitDegrees=0}
 {bone1Name=dyn_charm_a_2, bone2Name=dyn_charm_a_1, swing1LimitDegrees=20, swing2LimitDegrees=40, twistLimitDegrees=0}
 {bone1Name=dyn_charm_a_3, bone2Name=dyn_charm_a_2, swing1LimitDegrees=20, swing2LimitDegrees=40, twistLimitDegrees=0}
 {bone1Name=dyn_charm_a_4, bone2Name=dyn_charm_a_3, swing1LimitDegrees=20, swing2LimitDegrees=40, twistLimitDegrees=0}
 {bone1Name=dyn_charm_a_5, bone2Name=dyn_charm_a_4, swing1LimitDegrees=20, swing2LimitDegrees=40, twistLimitDegrees=0}
 {bone1Name=dyn_charm_b_1, bone2Name=dyn_charm_1, swing1LimitDegrees=20, swing2LimitDegrees=30, twistLimitDegrees=0}
 {bone1Name=dyn_charm_b_2, bone2Name=dyn_charm_b_1, swing1LimitDegrees=20, swing2LimitDegrees=30, twistLimitDegrees=0}
 {bone1Name=dyn_charm_b_3, bone2Name=dyn_charm_b_2, swing1LimitDegrees=20, swing2LimitDegrees=30, twistLimitDegrees=0}
 {bone1Name=dyn_charm_b_4, bone2Name=dyn_charm_b_3, swing1LimitDegrees=20, swing2LimitDegrees=30, twistLimitDegrees=0}
 {
  bone1Name=base_charm_1
  bone2Name=pack
  swing1LimitDegrees=45
  swing1Motion=Free
  swing2LimitDegrees=45
  swing2Motion=Free
  twistLimitDegrees=45
  twistMotion=Free
 }
]
